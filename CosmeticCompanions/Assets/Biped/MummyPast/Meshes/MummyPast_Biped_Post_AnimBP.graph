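# Post-process AnimBlueprint for the MummyPast biped cosmetic companion, applying corrective posing after the main animation.
# The input pose passes through a pose asset, translation constraints from deform thighs to feet, and additive foot offsets.
# The left foot offset requires the Stage2 (toon) outfit variant, checked at initialization; the right is always enabled.
# At LOD 0–1, left-eye blink curves drive the BodyScale morph; a blink curve-remap node sits unconnected.
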AnimBlueprint /CosmeticCompanions/Assets/Biped/MummyPast/Meshes/MummyPast_Biped_Post_AnimBP
target_skeleton=&/MummyPast_Biped_Skeleton
graphs [
 AnimGraph {nodes [
   Remap Curves From Mesh!(
    SourcePose:Pose Link Structure
    CurveExpressionsDataAsset:Curve Expressions Data Asset Object Reference=&/MummyPast_Blink_Remap
    SourceMeshComponent:Skeletal Mesh Component Weak object=Get mesh component() -> (mesh component:Skeletal Mesh Component Object Reference)
   ) -> (Pose:Pose Link Structure) {is=title AnimGraphNode_RemapCurvesFromMesh, bUseAttachedParent=true, expressionSource=DataAsset}
   n1 = Input Pose()
   n2 = Get PoseAsset() -> (PoseAsset:Pose Asset Object Reference)
   n3 = "Constraint - deform_thigh_l (T)"(
    ComponentPose=Local To Component(LocalPose=Pose!(PoseAsset:Pose Asset Object Reference=@n2, SourcePose:Pose Link Structure=@n1) -> (Pose:Pose Link Structure) {is=title AnimGraphNode_PoseBlendNode, blendOption=Linear})
    Alpha=1
    ConstraintWeights_0=1
   ) {boneToModify {boneName=deform_thigh_l}, constraintSetup [{transformType=Translation, targetBone {boneName=foot_l}}], constraintWeights [1]}
   n4 = "Constraint - deform_thigh_r (T)"(ComponentPose=@n3, Alpha=1, ConstraintWeights_0=1) {boneToModify {boneName=deform_thigh_r}, constraintSetup [{transformType=Translation, targetBone {boneName=foot_r}}], constraintWeights [1]}
   n5 = "Transform (Modify) Bone - Bone: foot_l"!(
    ComponentPose=@n4
    bAlphaBoolEnabled:bool=Get IsToonVariant() -> (IsToonVariant:bool)
    Translation="1.3,0,0"
    Rotation="0,0,0"
    Scale="1,1,1"
   ) -> (Pose) {alphaInputType=Bool, translationMode=BMM_Additive, translationSpace=BCS_BoneSpace, boneToModify {boneName=foot_l}, translation {x=1.3}}
   n6 = "Transform (Modify) Bone - Bone: foot_r"!(ComponentPose=@n5, bAlphaBoolEnabled:bool=True, Translation="-1.3,0,0", Rotation="0,0,0", Scale="1,1,1") -> (Pose) {alphaInputType=Bool, translationMode=BMM_Additive, translationSpace=BCS_BoneSpace, boneToModify {boneName=foot_r}, translation {x=-1.3}}
   Output Pose(Result=Component To Local(ComponentPose=@n6))
   Set members in Named Curve Value!(execute:Exec, StructRef:"Named Curve Value Structure (by ref)") -> (then:Exec, StructOut:"Named Curve Value Structure (by ref)") {is=title K2Node_SetFieldsInStruct}
  ]}
 EventGraph {nodes [
   Event BlueprintInitializeAnimation()
   > Set IsToonVariant(IsToonVariant:bool=HasVariantTagForChannel!(
     self:Fort Player Companion Anim Instance Object Reference=GetMainAnimInstance!(self:Companion Post Process Anim Instance Object Reference) -> (ReturnValue:Fort Player Companion Anim Instance Object Reference)
     ChannelTag:Gameplay Tag Structure=(TagName="Cosmetics.Variant.Channel.Outfit")
     VariantTag:Gameplay Tag Structure=(TagName="Cosmetics.Variant.Property.Stage2")
    ) -> (ReturnValue:bool)) -> (Output_Get:bool)
   Event BlueprintUpdateAnimation()
   > n1 = Branch(Condition="integer > integer"(A=GetLODLevel!(self:Anim Instance Object Reference) -> (ReturnValue:int=0), B=1))
   SetMorphTarget!(
    execute=@n1.else
    self:Anim Instance Object Reference
    MorphTargetName:Name=BodyScale
    Value:float="Max (Float)"(A=GetCurveValue(CurveName=L_blink_pose), B=GetCurveValue(CurveName=L_lower_lid_up_pose))
   ) -> (then)
  ]}
]
variables [
 IsToonVariant {type=boolean}
 mesh component {type=SkeletalMeshComponent}
 PoseAsset {type=PoseAsset, default=&/MummyPast_Biped_PoseAsset}
]
